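# Material function that outputs cubemap UVW coordinates faking a box-shaped room inset behind a surface, such as parallax interiors seen through building windows.
# Tiled UVs become per-cell positions on the box front, and the tangent-space camera vector drives a ray-box intersection giving the lookup vector.
# The optional Randomize Rotation static switch uses the parity of each tile index to mirror and swizzle that vector, varying rooms between tiles.
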MaterialFunction /Engine/Functions/Engine_MaterialFunctions01/Cubemaps/InteriorCubemap
graph {
 properties {
  bExposeToLibrary=true
  description=Interior Cubemap conforms a cubemap to be in the form of a box shape inset from the surface UVs. It is useful for giving parallax to rooms visible through windows on city buildings.
  libraryCategoriesText [Cubemaps]
 }
 nodes [
  n1 = FunctionInput() {bUsePreviewValueAsDefault=true, inputName=Tiling, inputType=FunctionInput_Vector2, sortPriority=1, previewValue=<4 4 0 1>}
  n2 = Multiply(A=FunctionInput(Preview=TextureCoordinate()) {bUsePreviewValueAsDefault=true, inputName=UVs, inputType=FunctionInput_Vector2}, B=@n1)
  n3 = MaterialFunctionCall(Float2=Ceil(None=@n2)) {materialFunction=&BreakOutFloat2Components}
  n4 = Constant() {r=3}
  Fmod(A=@n3.R, B=@n4)
  n5 = Multiply(
   A=Constant3Vector() {constant=<rgba -1 1 1 0>}
   B=Transform(None=CameraVectorWS()) {desc=Tangent, transformSourceType=TRANSFORMSOURCE_World, transformType=TRANSFORM_Tangent}
  )
  n6 = Divide(B=@n5) {desc=id}
  n7 = Frac(None=@n2) {desc=f}
  n8 = Constant2Vector() {g=-2, r=2}
  n9 = AppendVector(A=Subtract(A=Multiply(A=@n7, B=@n8), B=Constant2Vector() {g=-1, r=1}), B=Constant() {r=-1})
  n10 = MaterialFunctionCall(Float3=Subtract(A=Abs(None=@n6), B=Multiply(A=@n6, B=@n9)) {desc=k}) {materialFunction=&BreakOutFloat3Components}
  n11 = Min(A=Min(A=@n10.R, B=@n10.R), B=@n10.B)
  n12 = MaterialFunctionCall(Float3=Add(A=Multiply(A=@n11, B=@n5), B=@n9) {desc="pos.xyz += kMin * In.dir;"}) {materialFunction=&BreakOutFloat3Components}
  n13 = MaterialFunctionCall(X=@n12.B, Y=@n12.B, Z=@n12.B) {materialFunction=&MakeFloat3}
  n14 = Constant() {r=2}
  n15 = Fmod(A=@n3.G, B=@n14)
  n16 = Multiply(
   A=LinearInterpolate(A=Constant3Vector() {constant=<rgba -1 1 1 0>}, B=Constant3Vector() {constant=<rgba 1 -1 1 0>}, Alpha=Fmod(A=@n3.R, B=@n14))
   B=LinearInterpolate(A=Constant3Vector() {constant=<rgba 1 1 1 0>}, B=Constant3Vector() {constant=<rgba -1 1 1 0>}, Alpha=@n15)
  )
  n17 = Multiply(A=@n13.Result, B=@n16)
  n18 = MaterialFunctionCall(XYZ=@n17) {materialFunction=&Swizzle}
  n19 = FunctionInput(Preview=StaticBool()) {bUsePreviewValueAsDefault=true, inputName=Randomize Rotation, inputType=FunctionInput_StaticBool, sortPriority=2, previewValue=<4 4 0 1>}
  FunctionOutput(None=StaticSwitch(True=LinearInterpolate(A=@n17, B=@n18.YXZ, Alpha=@n15), False=@n13.Result, Value=@n19)) {outputName=UVW}
  LinearInterpolate(
   A=@n16
   B=MaterialFunctionCall(XYZ=@n16) {materialFunction=&Swizzle}.YXZ
   Alpha=Fmod(A=@n3.G, B=@n4)
  )
 ]
}
